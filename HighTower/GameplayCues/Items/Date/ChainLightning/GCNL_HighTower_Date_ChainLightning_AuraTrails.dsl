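; Looping gameplay cue for the HighTower Date chain lightning aura trails, which picks the effect color for the weapon variant.
; On loop start it casts the target to a Fortnite player pawn, then its current weapon to the Date combo weapon.
; It reads that weapon's IsUltra flag to set FX_Color to orange for Ultra or green otherwise, then leaves the rest to the parent cue.
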(event EventOnLoopingStart(EitherVFX-preferthisversion)
  (params (MyTarget :type Actor%20Object%20Reference) (Parameters :type Gameplay%20Cue%20Parameters%20Structure%20%28by%20ref%29) (ParticleComponents :type Array%20of%20FXSystem%20Component%20Object%20References) (AudioComponents :type Array%20of%20Audio%20Component%20Object%20References))
  (bind _asfort_player_pawn (Utilities|Casting|CastToFortPlayerPawn MyTarget))
  (bind _asb_high_tower_date_combo (Utilities|Casting|CastToB_HighTower_Date_Combo (Variables|/Script/FortniteGame.FortPawn|GetCurrentWeapon _asfort_player_pawn)))
  (if (Variables|/HighTower/Items/Date/ChainLightning/B_HighTower_Date_ChainLightning.B_HighTower_Date_ChainLightning_C|GetIsUltra _asb_high_tower_date_combo)
    (Variables|Default|SetFX_Color "(R=2.000000,G=0.520132,B=0.000000,A=1.000000)")
    (else
      (Variables|Default|SetFX_Color "(R=0.276909,G=0.708333,B=0.043574,A=1.000000)")))
  (|CallParentFunction MyTarget Parameters ParticleComponents AudioComponents))

(fn ConstructionScript
  (|CallParentFunction))
