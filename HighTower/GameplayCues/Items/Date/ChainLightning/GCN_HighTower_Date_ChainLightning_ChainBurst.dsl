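; Burst gameplay cue for the HighTower Date item's chain lightning, drawing the effect of one link between two actors.
; On burst, it aims the first particle component from the Instigator at the EffectCauser, and moves the second onto the EffectCauser facing back.
; The distance between them drives the first particle's Size (scaled by a Divisor variable), Offset and CylinderHeight parameters.
; A debug PrintString of that distance remains in the graph.
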(fn OnBurst(EitherVFX-preferthisversion)
  (params (MyTarget :type Actor%20Object%20Reference) (Parameters :type Gameplay%20Cue%20Parameters%20Structure%20%28by%20ref%29) (ParticleComponents :type Array%20of%20FXSystem%20Component%20Object%20References) (AudioComponents :type Array%20of%20Audio%20Component%20Object%20References) (BurstCameraShakeInstance :type Camera%20Shake%20Base%20Object%20Reference) (BurstDecalInstance :type Decal%20Actor%20Object%20Reference))
  (bind _returnvalue (Transformation|GetActorLocation (.Instigator (Ability|GameplayCue|BreakGameplayCueParameters Parameters))))
  (bind _returnvalue_2 (Math|Conversions|ToVector(Float) (Class|AbilityFunctions|GetDistanceBetweenActors (.Instigator (Ability|GameplayCue|BreakGameplayCueParameters Parameters)) (.EffectCauser (Ability|GameplayCue|BreakGameplayCueParameters Parameters)) (.Instigator (Ability|GameplayCue|BreakGameplayCueParameters Parameters)))))
  (bind (_normalizedmagnitude _rawmagnitude _effectcontext _matchedtagname _originaltag _aggregatedsourcetags _aggregatedtargettags _location _normal _instigator _effectcauser) (Ability|GameplayCue|BreakGameplayCueParameters Parameters))
  (bind _distance (Class|AbilityFunctions|GetDistanceBetweenActors _instigator _effectcauser _instigator))
  (bind _output (Utilities|Array|Get(acopy) ParticleComponents))
  ;[1 1 1] This is the Target
  (Utilities|IsValid _instigator
    (:Is%20Valid
      (bind _asactor (Utilities|Casting|CastToActor _effectcauser))
      (Transformation|SetWorldLocationAndRotation (Utilities|Array|Get(acopy) ParticleComponents 1) (Transformation|GetActorLocation _asactor) (Math|Rotator|FindLookatRotation (Transformation|GetActorLocation _asactor) _returnvalue) false false)
      ;[1 1 1] This is the Source
      (Transformation|SetWorldRotation _output (Math|Rotator|FindLookatRotation _returnvalue (Transformation|GetActorLocation _asactor)) false false)
      (Effects|Components|ParticleSystem|SetVectorParameter _output "Size" (/ _returnvalue_2 (Variables|Default|GetDivisor)))
      (Effects|Components|ParticleSystem|SetVectorParameter _output "Offset" _returnvalue_2)
      (Effects|Components|ParticleSystem|SetFloatParameter _output "CylinderHeight" _distance)
      (Development|PrintString 0 (Utilities|String|ToString(Float) _distance) true true "(R=0.000000,G=0.660000,B=1.000000,A=1.000000)" 3.0)
      (return))))
