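; Burst gameplay cue for the explosion of the HighTower Date chain lightning item.
; On burst it shows a local gunshot-type sound indicator at the target, with audible distance from a scalable float.
; It enables the particle's SpawnDecal parameter only when raw magnitude is nonzero and the effect causer lacks the vehicle and player pawn tags.
; Raw magnitude also doubles as the particle's uniform scale when above one.
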(event Custom|Explode%20Audio)

(fn OnBurst(EitherVFX-preferthisversion)
  (params (MyTarget :type Actor%20Object%20Reference) (Parameters :type Gameplay%20Cue%20Parameters%20Structure%20%28by%20ref%29) (ParticleComponents :type Array%20of%20FXSystem%20Component%20Object%20References) (AudioComponents :type Array%20of%20Audio%20Component%20Object%20References) (BurstCameraShakeInstance :type Camera%20Shake%20Base%20Object%20Reference) (BurstDecalInstance :type Decal%20Actor%20Object%20Reference))
  (bind (_normalizedmagnitude _rawmagnitude) (Ability|GameplayCue|BreakGameplayCueParameters Parameters))
  (bind (_normalizedmagnitude_2 _rawmagnitude_2 _effectcontext _matchedtagname _originaltag _aggregatedsourcetags _aggregatedtargettags _location _normal _instigator _effectcauser) (Ability|GameplayCue|BreakGameplayCueParameters Parameters))
  (bind _output (Utilities|Array|Get(acopy) ParticleComponents))
  (Fortnite|ShowSoundIndicatorLocal :WorldContextObject MyTarget :TrackedActor MyTarget :Location (Transformation|GetActorLocation MyTarget) :MaxAudibleDistance (ScalableFloat|GetValueatLevel (Variables|Default|GetRow_AVIndicatorDistance)) :IndicatorType "Gunshot" :Instigator MyTarget :Tint "(R=0.000000,G=0.000000,B=0.000000,A=0.000000)" :AttenuationMaxRangeMultiplier 1.0 :AttenuationIntensityMultiplier 1.0)
  (bind _returnvalue (GameplayTags|HasMatchingGameplayTag(Message) _effectcauser (Variables|Default|GetVehicle%20Tag))) ; Dont spawn decal if hitting on vehicles
  (bind _returnvalue_2 (GameplayTags|HasMatchingGameplayTag(Message) _effectcauser (Variables|Default|GetPlayerPawn%20Tag))) ; Dont spawn decal if hitting on players
  (Effects|Components|ParticleSystem|SetBoolParameter _output "SpawnDecal" (not (Math|Boolean|ORBoolean _returnvalue_2 _returnvalue (not (Widget|UEFN|ToBool(Double) _rawmagnitude_2)))))
  (Transformation|SetWorldScale3D _output (Math|Vector|SelectVector (Math|Conversions|ToVector(Float) _rawmagnitude) "1.000000,1.000000,1.000000" (> _rawmagnitude 1.0)))
  (return))
